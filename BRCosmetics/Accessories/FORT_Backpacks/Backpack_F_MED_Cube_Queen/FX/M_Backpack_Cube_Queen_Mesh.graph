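# Translucent, unlit, two-sided Niagara mesh particle material for flowing energy effects on the Cube Queen backpack.
# Panning rainbow and flame-tile textures, distorted by scrolling smoke noise, combine with a three-color lerp to build the tinted color.
# A smoke-distorted scrolling mask tints the center through its green channel and shapes opacity through red with a vertical gradient fade.
# Particle color modulates color and alpha, with parameters controlling saturation, contrast and intensity.
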Material /BRCosmetics/Accessories/FORT_Backpacks/Backpack_F_MED_Cube_Queen/FX/M_Backpack_Cube_Queen_Mesh
graph {
 properties {
  bUsedWithNiagaraMeshParticles=true
  blendMode=BLEND_Translucent
  pixelDepthOffsetMode=PDOM_Legacy
  shadingModel=MSM_Unlit
  twoSided=true
  usedShadingModels=Unlit
 }
 nodes [
  Add()
  Constant2Vector() {g=0.2, r=1}
  n1 = Multiply(
   A=TextureCoordinate() {vTiling=0.2}
   B=ComponentMask(None=VectorParameter() {parameterName=ColorUV, defaultValue=<rgba 1 0.2 0 0>}.RGB) {g=true, r=true}
  )
  n2 = TextureCoordinate() {vTiling=0.2}
  n3 = Multiply(A=TextureSample(UVs=Panner(Coordinate=@n2) {speedY=-0.2}) {texture=&T_Inky_Smoke_Tile}.R) {constB=0.2}
  n4 = TextureSample(UVs=Add(A=Panner(Coordinate=@n1) {speedY=0.2}, B=@n3)) {texture=&T_Tile_RainbowColor}
  n5 = ScalarParameter() {defaultValue=1, parameterName=TexMultiply}
  n6 = TextureCoordinate() {vTiling=0.2}
  n7 = Multiply(A=TextureSample(UVs=Panner(Coordinate=@n6) {speedY=0.15}) {texture=/Game/Global/Textures/Default/Tileable_Patterns/T_Thick_Smoke_05}.R) {constB=0.2}
  n8 = MaterialFunctionCall(Float2=TextureCoordinate()) {materialFunction=&BreakOutFloat2Components}
  n9 = Power(Base=OneMinus(None=@n8.G))
  n10 = Multiply(
   A=MaterialFunctionCall(X=@n8.R, Y=OneMinus(None=@n9)) {materialFunction=&MakeFloat2}.Result
   B=ComponentMask(None=VectorParameter() {parameterName=TileUV, defaultValue=<rgba 1 0.2 0 0>}.RGB) {g=true, r=true}
  )
  n11 = TextureSampleParameter2D(UVs=Add(A=Multiply(A=@n7, B=@n3), B=Panner(Coordinate=@n10) {speedY=0.2})) {
   parameterName=Tile
   samplerType=SAMPLERTYPE_LinearGrayscale
   texture=/BRCosmetics/Weapons/FORT_Melee/Pickaxe_Cube_Ninja_Male/FX/T_FluidFlameTile
  }
  n12 = Power(Base=@n11.R, Exp=ScalarParameter() {defaultValue=1, parameterName=Tile Power})
  n13 = Multiply(
   A=VectorParameter() {group=Color, parameterName=Detail_Color3, defaultValue=<rgba 0.2521 0.0245 1 1>}.RGB
   B=Add(A=@n12, B=ScalarParameter() {parameterName=Detail_Color3_Add})
  )
  n14 = VectorParameter() {group=Color, parameterName=Detail_Color2, defaultValue=<rgba 0.2521 0.0245 1 1>}
  n15 = VectorParameter() {group=Color, parameterName=Detail_Color1, defaultValue=<rgba 0 0.6333 2 1>}
  n16 = MaterialFunctionCall(In=@n12, Contrast=ScalarParameter() {defaultValue=0.5, parameterName=Tile Contrast}) {materialFunction=&CheapContrast}
  n17 = ScalarParameter() {parameterName=Tile Add}
  n18 = Multiply(A=Saturate(None=Add(A=@n16.Result, B=@n17)))
  n19 = Multiply(
   A=Desaturation(None=Multiply(A=@n4.RGB, B=@n5), Fraction=ScalarParameter() {parameterName=TexSaturation})
   B=Desaturation(
    None=MaterialFunctionCall(A=@n13, B=@n14.RGB, C=@n15.RGB, Alpha=@n18) {materialFunction=&Lerp_3Color}.Result
    Fraction=ScalarParameter() {parameterName=DetailColorSaturation}
   )
  ) {constB=10}
  n20 = VectorParameter() {group=Color, parameterName=Center_Color, defaultValue=<rgba 1 1 1 1>}
  n21 = Panner(Coordinate=TextureCoordinate() {uTiling=2, vTiling=0.2}) {speedY=0.15}
  n22 = Add(
   A=Multiply(A=TextureSample(UVs=@n21) {texture=&T_Inky_Smoke_Tile}.R) {constB=0.4}
   B=Panner(Coordinate=TextureCoordinate()) {speedX=-0.2}
  )
  n23 = MaterialFunctionCall!(TexCoord=@n22, ClampChannel=Constant2Vector() {g=1}) {materialFunction=/Game/Effects/Fort_Effects/MaterialFunctions/MF_FX_TilingMethod}
  n24 = TextureSampleParameter2D(UVs=@n22) {parameterName=MaskTex, texture=/Game/Accessories/FORT_Backpacks/Backpack_F_MED_Cube_Queen/FX/T_Backpack_Cube_Queen_Mask}
  n25 = Multiply(
   A=Power(Base=Multiply(A=@n23.Result, B=@n24.G), Exp=ScalarParameter() {defaultValue=3, parameterName=Center_Power}) {constExponent=3}
   B=ScalarParameter() {defaultValue=1, parameterName=Center_Multiply}
  )
  n26 = Desaturation(
   None=LinearInterpolate(A=@n19, B=Multiply(A=@n19, B=@n20.RGB), Alpha=Saturate(None=@n25))
   Fraction=ScalarParameter() {group=Final, parameterName=FinalSaturation}
  )
  n27 = ScalarParameter() {defaultValue=1, group=Final, parameterName=FinalMultiply}
  n28 = ParticleColor()
  MaterialFunctionCall(Emissive=Multiply(A=Multiply(A=@n26, B=@n27), B=@n28.RGB)) {materialFunction=&MF_FX_EmissiveAdjustments}
  n29 = Multiply(A=@n23.Result, B=@n24.R)
  n30 = ScalarParameter() {defaultValue=1, parameterName=Alpha}
  n31 = Multiply(
   A=OneMinus(None=MaterialFunctionCall() {materialFunction=&LinearGradient}.VGradient)
   B=ScalarParameter() {defaultValue=1, parameterName=LinearGradientMultiply}
  )
  Multiply(A=@n28.A, B=Multiply(A=Multiply(A=@n29, B=@n30), B=Saturate(None=@n31)))
  ScalarParameter() {parameterName=FinalAdd}
  TextureCoordinate()
 ]
}
