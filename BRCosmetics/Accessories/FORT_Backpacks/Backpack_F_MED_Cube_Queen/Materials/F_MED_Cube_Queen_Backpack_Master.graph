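# Master material for the Cube Queen backpack, layering animated crystals, rings and glowing plasma between character PreFX and PostFX functions.
# Vertex color supplies element masks and random seeds for sine-driven translation and rotation through World Position Offset, with normals rotated to match.
# Emissive combines crystal flash, plasma and base color, scaled by time of day, alongside an optional anodized-metal effect.
# A separate Blueprint-parameter bone-pivot rotation branch feeds no other node.
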Material /BRCosmetics/Accessories/FORT_Backpacks/Backpack_F_MED_Cube_Queen/Materials/F_MED_Cube_Queen_Backpack_Master
graph {
 properties {
  bAutomaticallySetUsageInEditor=false
  bTangentSpaceNormal=false
  bUseMaterialAttributes=true
  bUsedWithMeshDeformer=true
  bUsedWithSkeletalMesh=true
  pixelDepthOffsetMode=PDOM_Legacy
  previewMesh=/BRCosmetics/Accessories/FORT_Backpacks/Backpack_F_MED_Cube_Queen/Meshes/F_MED_Cube_Queen_Pack
  shadingModel=MSM_FromMaterialExpression
  subsurfaceProfile=&SS_HeroSkin_02
  usedShadingModels=Default Lit | Subsurface Profile
 }
 nodes [
  Constant()
  n1 = MaterialFunctionCall() {materialFunction=&BaseFN_CharacterFunction_PreFX}
  n2 = GetMaterialAttributes(None=@n1.Result) {attributeGetTypes [#1, #2]}
  n3 = SetMaterialAttributes!(
   MaterialAttributes=NamedRerouteDeclaration(None=@n2.MaterialAttributes) {name=MatAttrib_IN, nodeColor=<rgba 0 0.9766 1 1>}
   World Position Offset=NamedRerouteUsage()
   Normal=NamedRerouteUsage()
   Emissive Color=NamedRerouteUsage()
   Base Color=NamedRerouteUsage()
  ) {attributeSetTypes [
    #3
    #1
    #4
    #2
   ]}
  n4 = MaterialFunctionCall(AO=@n1.AO, Distance Blend=@n1.AO, Fresnel=@n1.AO, Gmap=@n1.AO, Pre Skinned Local Position=@n1.AO, Material=@n3) {materialFunction=&BaseFN_CharacterFunction_PostFX}
  MaterialFunctionCall(In=@n4.Result) {materialFunction=&CINE_MF_CharPostFX}
  n5 = Add(
   A=Multiply(A=Time(), B=ScalarParameter() {defaultValue=0.5, group=CubeQueenAnimTranslation, parameterName=CrystalsTranslationSpeed})
   B=Multiply(A=NamedRerouteUsage(), B=ScalarParameter() {defaultValue=5, group=CubeQueenAnimTranslation, parameterName=CrystalsTranslationRandomSpeed})
  )
  n6 = Power(
   Base=ConstantBiasScale(None=Sine(None=@n5))
   Exp=ScalarParameter() {defaultValue=0.33, group=CubeQueenAnimTranslation, parameterName=CrystalsTranslationPow}
  )
  n7 = Multiply(A=@n6, B=ScalarParameter() {defaultValue=2.5, group=CubeQueenAnimTranslation, parameterName=CrystalsTranslationAmplitude})
  n8 = Multiply(A=VertexColor()) {constB=100}
  n9 = Normalize(VectorInput=Multiply(A=Subtract(A=@n8) {constB=20}, B=VectorParameter() {defaultValue=<rgba 1 0 1 1>}.RGB))
  n10 = Time()
  n11 = Multiply(A=NamedRerouteUsage(), B=ScalarParameter() {defaultValue=0.5, group=CubeQueenAnimTranslation, parameterName=TranslationRandomSpeed})
  n12 = ConstantBiasScale(None=Sine(None=Multiply(A=@n10, B=@n11) {constB=0.5}))
  n13 = Multiply(
   A=ScalarParameter() {defaultValue=2.5, group=CubeQueenAnimTranslation, parameterName=TranslationAmplitude}
   B=Saturate(None=InverseLinearInterpolate(Value=@n12) {constA=0.45, constB=0.55})
  )
  NamedRerouteDeclaration(None=LinearInterpolate(A=Multiply(A=@n7, B=@n9), B=Multiply(A=@n9, B=@n13), Alpha=NamedRerouteUsage()) {constB=0}) {name=PosOffset, nodeColor=<rgba 1 0 0.4922 1>}
  n14 = VertexColor()
  n15 = Constant() {r=0.2}
  n16 = Constant() {r=0.01}
  n17 = Constant() {r=1}
  NamedRerouteDeclaration(None=OneMinus(None=SphereMask(A=@n14.A, B=@n15, Radius=@n16, Hardness=@n17))) {name=CrystalsMask, nodeColor=<rgba 0 0.0156 1 1>}
  NamedRerouteDeclaration(None=@n1.Fresnel) {name=Fresnel_IN, nodeColor=<rgba 1 0 0.75 1>}
  NamedRerouteDeclaration(None=@n2.Normal) {name=Normal_IN, nodeColor=<rgba 0 0.9766 1 1>}
  NamedRerouteDeclaration(None=@n6) {name=CrystalsTranslation, nodeColor=<rgba 0 1 0.3203 1>}
  n18 = Fresnel(
   ExponentIn=ScalarParameter() {defaultValue=1, group=Anodized Metal, parameterName=Anodized_FresnelExponent}
   BaseReflectFractionIn=ScalarParameter() {defaultValue=3, group=Anodized Metal, parameterName=Anodized_ReflectFraction}
   Normal=PixelNormalWS()
  ) {baseReflectFraction=0.5, exponent=3}
  n19 = ScalarParameter() {group=Anodized Metal, parameterName=Anodized_ColorOffset}
  n20 = Multiply(
   A=TextureSample(UVs=Add(A=@n18, B=@n19)) {texture=/Game/Global/Textures/Default/Masks_and_Gradients/3pxRGBImage}.RGB
   B=VectorParameter() {group=Anodized Metal, parameterName=AnodizationFilter, defaultValue=<rgba 1 0 0.5305 0>}.RGB
  )
  n21 = Multiply(A=Constant() {r=4}, B=@n20)
  n22 = StaticComponentMaskParameter(None=AtmosphericLightVector()) {
   defaultB=true
   defaultG=true
   defaultR=true
   desc=Use RG for 2D vectors dot product. else use RGB
   group=Anodized Metal
   parameterName=AnodizedLdotR_Components
  }
  n23 = StaticComponentMaskParameter(None=ReflectionVectorWS()) {
   defaultB=true
   defaultG=true
   defaultR=true
   desc=Use RG for 2D vectors dot product. else use RGB
   group=Anodized Metal
   parameterName=AnodizedLdotR_Components
  }
  n24 = DotProduct(A=@n22, B=@n23)
  n25 = ScalarParameter() {defaultValue=55, group=Anodized Metal, parameterName=Anodized_pow}
  n26 = Saturate(None=Multiply(A=Power(Base=@n24, Exp=@n25), B=Constant() {r=512}))
  n27 = Multiply(
   A=Desaturation(None=LinearInterpolate(A=@n20, B=@n21, Alpha=@n26), Fraction=ScalarParameter() {group=Anodized Metal, parameterName=Anodized_Desat})
   B=Power(Base=NamedRerouteUsage(), Exp=ScalarParameter() {group=Anodized Metal, parameterName=AnodizedEffectFresnelMask})
  )
  n28 = ScalarParameter() {defaultValue=1, group=Anodized Metal, parameterName=AnodizedEffectIntensity}
  NamedRerouteDeclaration(None=StaticSwitchParameter(True=Multiply(A=@n27, B=@n28), False=Constant()) {group=Anodized Metal, parameterName=UseAnodizedEffect}) {name=AnodizedMetalColors, nodeColor=<rgba 1 0 0.4219 1>}
  MaterialExpressionConstant_2 = Constant() {r=1}
  MaterialExpressionComponentMask_2 = ComponentMask(None=Multiply(A=VertexColor(), B=ScalarParameter() {defaultValue=150, group=CubeQueenAnim, parameterName=Randomness})) {b=true}
  n29 = Custom!(x=@MaterialExpressionConstant_2, magic=@MaterialExpressionComponentMask_2) {
   code="RandFast(x,magic);"
   outputType=CMOT_Float1
   inputs [
    {
     inputName=x
     input {expression=&self:MaterialExpressionConstant_2}
    }
    {
     inputName=magic
     input {expression=&self:MaterialExpressionComponentMask_2}
    }
   ]
  }
  NamedRerouteDeclaration(None=@n29) {name=RandomValuePerElement, nodeColor=<rgba 0.4531 0 1 1>}
  NamedRerouteDeclaration(None=SphereMask(A=VertexColor().A, B=Constant() {r=0.5}, Radius=Constant() {r=0.01}, Hardness=Constant() {r=1})) {name=InnerRingMask, nodeColor=<rgba 0 0.0156 1 1>}
  n30 = NamedRerouteUsage()
  n31 = ScalarParameter() {defaultValue=0.5, group=CubeQueenAnimRotation, parameterName=RotationRandomSpeed}
  n32 = Multiply(A=Time(), B=ScalarParameter() {defaultValue=0.1, group=CubeQueenAnimRotation, parameterName=MainRotationSpeed})
  n33 = Multiply(
   A=Time()
   B=LinearInterpolate(
    A=ScalarParameter() {defaultValue=0.15, group=CubeQueenAnimRotation, parameterName=SecondaryRotationSpeed}
    B=ScalarParameter() {defaultValue=0.4, group=CubeQueenAnimRotation, parameterName=SecondaryRotationSpeed-InnerRing}
    Alpha=NamedRerouteUsage()
   )
  )
  n34 = Multiply(
   A=SmoothStep(Value=Sine(None=@n33)) {constMax=0.2, constMin=-0.2}
   B=ScalarParameter() {defaultValue=0.5, group=CubeQueenAnimRotation, parameterName=SecondaryRotationAmplitude}
  )
  n35 = LinearInterpolate(
   A=Multiply(A=Time(), B=ScalarParameter() {defaultValue=0.005, group=CubeQueenAnimRotation, parameterName=CrystalsRotationSpeed})
   B=Multiply(A=Multiply(A=@n30, B=@n31), B=Add(A=@n32, B=@n34)) {constA=1}
   Alpha=NamedRerouteUsage()
  )
  n36 = RotateAboutAxis(
   NormalizedRotationAxis=Constant3Vector() {constant=<rgba 0 1 0 0>}
   RotationAngle=@n35
   PivotPoint=Constant()
   Position=Add(A=PreSkinnedPosition(), B=NamedRerouteUsage())
  )
  n37 = NamedRerouteUsage()
  NamedRerouteDeclaration(None=Transform(None=Add(A=@n36, B=@n37)) {transformSourceType=TRANSFORMSOURCE_Local}) {name=WPO_OUT, nodeColor=<rgba 0 0.4844 1 1>}
  n38 = Transform(None=NamedRerouteUsage())
  n39 = RotateAboutAxis(
   NormalizedRotationAxis=Transform(None=Constant3Vector() {constant=<rgba 0 1 0 0>}) {transformSourceType=TRANSFORMSOURCE_Local}
   RotationAngle=@n35
   PivotPoint=Constant()
   Position=@n38
  ) {desc=wo}
  NamedRerouteDeclaration(None=Normalize(VectorInput=Add(A=@n39, B=@n38))) {name=Normal_OUT, nodeColor=<rgba 0 0.7422 1 1>}
  n40 = OneMinus(None=NamedRerouteUsage())
  n41 = ScalarParameter() {defaultValue=5, group=CubeQueenShading, parameterName=CrystalFlashPow}
  n42 = NamedRerouteUsage()
  n43 = ComponentMask(None=@n42) {r=true}
  n44 = Multiply(
   A=Multiply(A=Power(Base=@n40, Exp=@n41), B=Multiply(A=@n43, B=@n43))
   B=VectorParameter() {group=CubeQueenShading, parameterName=CrystalFlashColor, defaultValue=<rgba 6.5696 1.5312 7 0>}.RGB
  )
  n45 = OneMinus(None=NamedRerouteUsage())
  n46 = Multiply(A=NamedRerouteUsage(), B=ScalarParameter() {defaultValue=0.5, group=CubeQueenShading, parameterName=BaseColorAsEmissive})
  n47 = ScalarParameter() {group=CubeQueenShading, parameterName=BaseColorAsEmissive_Desat}
  n48 = Multiply(
   A=MaterialFunctionCall!(NoiseSpeed, GradientTexture=OneMinus(None=@n43)) {materialFunction=/Game/Characters/Player/Female/Medium/Bodies/F_MED_Cube_Queen/Materials/MF_Cube_QueenPlasma}.Result
   B=ScalarParameter() {defaultValue=1, group=CubeQueenShading, parameterName=PlasmaEmissive}
  )
  n49 = NamedRerouteUsage()
  n50 = OneMinus(None=Power(
    Base=ComponentMask(None=@n49) {r=true}
    Exp=ScalarParameter() {defaultValue=0.05, group=CubeQueenShading, parameterName=PlasmaFresnelExponent}
   ))
  n51 = ScalarParameter() {group=CubeQueenShading, parameterName=MinPlasmaBrightness}
  n52 = Add(A=Add(A=Multiply(A=@n44, B=@n45), B=Desaturation(None=@n46, Fraction=@n47)), B=Max(A=Multiply(A=@n48, B=@n50), B=Multiply(A=@n42, B=@n51)))
  n53 = NamedRerouteUsage()
  n54 = ScalarParameter() {defaultValue=0.6666, group=CubeQueenShading, parameterName=CQ_Emissive_MorningEvening}
  n55 = ScalarParameter() {defaultValue=1, group=CubeQueenShading, parameterName=CQ_Emissive_Day}
  n56 = ScalarParameter() {defaultValue=0.6666, group=CubeQueenShading, parameterName=CQ_Emissive_Night}
  NamedRerouteDeclaration(None=Multiply(
    A=Multiply(A=@n52, B=@n53)
    B=MaterialFunctionCall(Value_Morning=@n54, Value_Day=@n55, Value_Evening=@n54, Value_Night=@n56) {materialFunction=&Scalar_Time_of_Day}.Result
   )) {name=Emissive_OUT, nodeColor=<rgba 0 1 0.6953 1>}
  n57 = Multiply(A=NamedRerouteUsage(), B=ScalarParameter() {defaultValue=0.1, group=CubeQueenShading, parameterName=BaseColor})
  n58 = ScalarParameter() {group=CubeQueenShading, parameterName=BaseColor_Desat}
  NamedRerouteDeclaration(None=Add(A=Desaturation(None=@n57, Fraction=@n58), B=NamedRerouteUsage())) {name=BaseColor_OUT, nodeColor=<rgba 1 0.375 0 1>}
  NamedRerouteDeclaration(None=@n2.Base Color) {name=BaseColor_IN, nodeColor=<rgba 0.8516 0 1 1>}
  NamedRerouteDeclaration(None=@n1.Gmap) {name=Gmap_IN, nodeColor=<rgba 0.5 1 0 1>}
  n59 = VectorParameter() {group=TRS, parameterName=BP_RightVector, defaultValue=<rgba 0 1 0 0>}
  n60 = ScalarParameter() {defaultValue=0.1, parameterName=spd}
  n61 = Multiply(A=ScalarParameter() {group=TRS, parameterName=BP_DeltaTime}, B=@n60)
  n62 = Subtract(
   A=VectorParameter() {group=TRS, parameterName=BP_BonePos}.RGB
   B=Transform(None=VectorParameter() {group=TRS, parameterName=BoneOffset}.RGB) {transformSourceType=TRANSFORMSOURCE_Local}
  ) {constB=0}
  n63 = WorldPosition()
  n64 = VectorParameter() {group=TRS, parameterName=BP_PreviousRightVector, defaultValue=<rgba 0 1 0 0>}
  n65 = Multiply(A=ScalarParameter() {group=TRS, parameterName=BP_PreviousDeltaTime}, B=@n60)
  n66 = Subtract(
   A=VectorParameter() {group=TRS, parameterName=BP_PreviousBonePos}.RGB
   B=Transform(None=VectorParameter() {group=TRS, parameterName=BoneOffset, defaultValue=<rgba 0 0 1.4 0>}.RGB) {transformSourceType=TRANSFORMSOURCE_Local}
  ) {constB=0}
  n67 = WorldPosition()
  n68 = Transform(None=NamedRerouteUsage())
  n69 = RotateAboutAxis(NormalizedRotationAxis=@n59.RGB, RotationAngle=@n61, PivotPoint=Constant(), Position=@n68) {desc=wo}
  SetMaterialAttributes!(
   MaterialAttributes=NamedRerouteUsage()
   World Position Offset=PreviousFrameSwitch(
    Current Frame=RotateAboutAxis(NormalizedRotationAxis=@n59.RGB, RotationAngle=@n61, PivotPoint=@n62, Position=@n63.XYZ)
    Previous Frame=RotateAboutAxis(NormalizedRotationAxis=@n64.RGB, RotationAngle=@n65, PivotPoint=@n66, Position=@n67.XYZ)
   )
   Normal=Normalize(VectorInput=Add(A=@n69, B=@n68))
  ) {attributeSetTypes [#3, #1]}
 ]
}
