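# Animation Blueprint for the Cube Queen backpack cosmetic, adding springy floating motion to its pieces on top of the pose copied from a mesh.
# Nine chained Anim Dynamics nodes simulate dyn_01 to dyn_09 relative to the pack bone, using gravity-free linear springs with tight travel limits.
# A time-based rotator and an additive JNT_rotator bone transform are present, but that transform is not wired into the output pose.
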AnimBlueprint /BRCosmetics/Accessories/FORT_Backpacks/Backpack_F_MED_Cube_Queen/Meshes/F_MED_Cube_Queen_Pack_AnimBP
target_skeleton=&/F_MED_Cube_Queen_Pack_Skeleton
graphs [
 AnimGraph {nodes [
   "Transform (Modify) Bone - Bone: JNT_rotator"(Alpha=1, Translation="0,0,0", Rotation=Get rotator() -> (rotator:Rotator), Scale="1,1,1") {
    rotationMode=BMM_Additive
    rotationSpace=BCS_ParentBoneSpace
    translationMode=BMM_Additive
    translationSpace=BCS_ParentBoneSpace
    alphaBoolBlend {blendOption=Linear}
    boneToModify {boneName=JNT_rotator}
   }
   n1 = Anim Dynamics - Bone: dyn_01(ComponentPose=Local To Component(LocalPose=Copy Pose From Mesh()), Alpha=1, ExternalForce="0,0,0") {
    bGravityOverrideInSimSpace=true
    bLinearSpring=true
    gravityScale=0
    linearDampingOverride=0.99
    linearSpringConstant=55
    alphaBoolBlend {blendOption=Linear}
    boundBone {boneName=dyn_01}
    componentAppliedLinearAccClamp=<11 11 11>
    componentLinearAccScale=<0.05 0.05 0.05>
    componentLinearVelScale=<0.9 0.9 0.9>
    physicsBodyDefinitions [{
      sphereCollisionRadius=0
      boundBone {boneName=dyn_01}
      boxExtents=<175 175 175>
      constraintSetup {angularLimitsMax=<0 0 0>, angularLimitsMin=<0 0 0>, angularTarget=<0 0 0>, linearAxesMax=<0.6 0.6 0.6>, linearAxesMin=<-0.5 -0.5 -0.5>}
      localJointOffset=<0 0 0>
     }]
    relativeSpaceBone {boneName=pack}
   }
   n2 = Anim Dynamics - Bone: dyn_02(ComponentPose=@n1, Alpha=1, ExternalForce="0,0,0") {
    bGravityOverrideInSimSpace=true
    bLinearSpring=true
    gravityScale=0
    linearDampingOverride=0.98
    linearSpringConstant=125
    alphaBoolBlend {blendOption=Linear}
    boundBone {boneName=dyn_02}
    componentAppliedLinearAccClamp=<12.5 12.5 12.5>
    componentLinearAccScale=<0.05 0.05 0.05>
    componentLinearVelScale=<0.9 0.9 0.9>
    physicsBodyDefinitions [{
      sphereCollisionRadius=0
      boundBone {boneName=dyn_02}
      boxExtents=<100 100 100>
      constraintSetup {angularLimitsMax=<0 0 0>, angularLimitsMin=<0 0 0>, angularTarget=<0 0 0>, linearAxesMax=<1 0.5 1>, linearAxesMin=<-0.5 -0.5 -0.5>}
      localJointOffset=<-0.5 0 -0.5>
     }]
    relativeSpaceBone {boneName=pack}
   }
   n3 = Anim Dynamics - Bone: dyn_03(ComponentPose=@n2, Alpha=1, ExternalForce="0,0,0") {
    bGravityOverrideInSimSpace=true
    bLinearSpring=true
    gravityScale=0
    linearDampingOverride=0.99
    linearSpringConstant=65
    alphaBoolBlend {blendOption=Linear}
    boundBone {boneName=dyn_03}
    componentAppliedLinearAccClamp=<15 15 15>
    componentLinearAccScale=<0.05 0.05 0.05>
    componentLinearVelScale=<0.9 0.9 0.9>
    physicsBodyDefinitions [{
      sphereCollisionRadius=0
      boundBone {boneName=dyn_03}
      boxExtents=<160 160 160>
      constraintSetup {angularLimitsMax=<0 0 0>, angularLimitsMin=<0 0 0>, angularTarget=<0 0 0>, linearAxesMax=<0.5 0.5 0.5>, linearAxesMin=<-0.5 -0.5 -0.5>}
      localJointOffset=<-1 -0.25 -0.5>
     }]
    relativeSpaceBone {boneName=pack}
   }
   n4 = Anim Dynamics - Bone: dyn_04(ComponentPose=@n3, Alpha=1, ExternalForce="0,0,0") {
    bGravityOverrideInSimSpace=true
    bLinearSpring=true
    gravityScale=0
    linearDampingOverride=0.99
    linearSpringConstant=65
    alphaBoolBlend {blendOption=Linear}
    boundBone {boneName=dyn_04}
    componentAppliedLinearAccClamp=<11 11 11>
    componentLinearAccScale=<0.05 0.05 0.05>
    componentLinearVelScale=<0.9 0.9 0.9>
    physicsBodyDefinitions [{
      sphereCollisionRadius=0
      boundBone {boneName=dyn_04}
      boxExtents=<160 160 160>
      constraintSetup {angularLimitsMax=<0 0 0>, angularLimitsMin=<0 0 0>, angularTarget=<0 0 0>, linearAxesMax=<0.5 0.5 0.5>, linearAxesMin=<-0.5 -0.5 -0.5>}
      localJointOffset=<-0.75 -0.25 -0.75>
     }]
    relativeSpaceBone {boneName=pack}
   }
   n5 = Anim Dynamics - Bone: dyn_05(ComponentPose=@n4, Alpha=1, ExternalForce="0,0,0") {
    bGravityOverrideInSimSpace=true
    bLinearSpring=true
    gravityScale=0
    linearDampingOverride=0.98
    linearSpringConstant=65
    alphaBoolBlend {blendOption=Linear}
    boundBone {boneName=dyn_05}
    componentAppliedLinearAccClamp=<16 16 16>
    componentLinearAccScale=<0.05 0.05 0.05>
    componentLinearVelScale=<0.9 0.9 0.9>
    physicsBodyDefinitions [{
      sphereCollisionRadius=0
      boundBone {boneName=dyn_05}
      boxExtents=<150 150 150>
      constraintSetup {angularLimitsMax=<0 0 0>, angularLimitsMin=<0 0 0>, angularTarget=<0 0 0>, linearAxesMax=<0.5 0.5 0.5>, linearAxesMin=<-0.5 -0.5 -0.5>}
      localJointOffset=<-0.75 -0.25 -0.75>
     }]
    relativeSpaceBone {boneName=pack}
   }
   n6 = Anim Dynamics - Bone: dyn_06(ComponentPose=@n5, Alpha=1, ExternalForce="0,0,0") {
    bGravityOverrideInSimSpace=true
    bLinearSpring=true
    gravityScale=0
    linearDampingOverride=0.98
    linearSpringConstant=75
    alphaBoolBlend {blendOption=Linear}
    boundBone {boneName=dyn_06}
    componentAppliedLinearAccClamp=<12.5 12.5 12.5>
    componentLinearAccScale=<0.05 0.05 0.05>
    componentLinearVelScale=<0.9 0.9 0.9>
    physicsBodyDefinitions [{
      sphereCollisionRadius=0
      boundBone {boneName=dyn_06}
      boxExtents=<125 125 125>
      constraintSetup {angularLimitsMax=<0 0 0>, angularLimitsMin=<0 0 0>, angularTarget=<0 0 0>, linearAxesMax=<1 0.5 1>, linearAxesMin=<-0.5 -0.5 -0.5>}
      localJointOffset=<-0.5 0 -0.5>
     }]
    relativeSpaceBone {boneName=pack}
   }
   n7 = Anim Dynamics - Bone: dyn_07(ComponentPose=@n6, Alpha=1, ExternalForce="0,0,0") {
    bGravityOverrideInSimSpace=true
    bLinearSpring=true
    gravityScale=0
    linearDampingOverride=0.99
    linearSpringConstant=45
    alphaBoolBlend {blendOption=Linear}
    boundBone {boneName=dyn_07}
    componentAppliedLinearAccClamp=<13 13 13>
    componentLinearAccScale=<0.05 0.05 0.05>
    componentLinearVelScale=<0.95 0.95 0.95>
    physicsBodyDefinitions [{
      sphereCollisionRadius=0
      boundBone {boneName=dyn_07}
      boxExtents=<180 180 180>
      constraintSetup {angularLimitsMax=<0 0 0>, angularLimitsMin=<0 0 0>, angularTarget=<0 0 0>, linearAxesMax=<0.6 0.6 0.6>, linearAxesMin=<-0.5 -0.5 -0.5>}
      localJointOffset=<-1 0 -0.2>
     }]
    relativeSpaceBone {boneName=pack}
   }
   n8 = Anim Dynamics - Bone: dyn_08(ComponentPose=@n7, Alpha=1, ExternalForce="0,0,0") {
    bGravityOverrideInSimSpace=true
    bLinearSpring=true
    gravityScale=0
    linearDampingOverride=0.99
    linearSpringConstant=45
    alphaBoolBlend {blendOption=Linear}
    boundBone {boneName=dyn_08}
    componentAppliedLinearAccClamp=<13 13 13>
    componentLinearAccScale=<0.05 0.05 0.05>
    componentLinearVelScale=<0.95 0.95 0.95>
    physicsBodyDefinitions [{
      sphereCollisionRadius=0
      boundBone {boneName=dyn_08}
      boxExtents=<165 165 165>
      constraintSetup {angularLimitsMax=<0 0 0>, angularLimitsMin=<0 0 0>, angularTarget=<0 0 0>, linearAxesMax=<0.6 0.6 0.6>, linearAxesMin=<-0.5 -0.5 -0.5>}
      localJointOffset=<-0.8 0 -0.3>
     }]
    relativeSpaceBone {boneName=pack}
   }
   n9 = Anim Dynamics - Bone: dyn_09(ComponentPose=@n8, Alpha=1, ExternalForce="0,0,0") {
    bGravityOverrideInSimSpace=true
    bLinearSpring=true
    gravityScale=0
    linearDampingOverride=0.99
    linearSpringConstant=65
    alphaBoolBlend {blendOption=Linear}
    boundBone {boneName=dyn_09}
    componentAppliedLinearAccClamp=<15 15 15>
    componentLinearAccScale=<0.05 0.05 0.05>
    componentLinearVelScale=<0.9 0.9 0.9>
    physicsBodyDefinitions [{
      sphereCollisionRadius=0
      boundBone {boneName=dyn_09}
      boxExtents=<160 160 160>
      constraintSetup {angularLimitsMax=<0 0 0>, angularLimitsMin=<0 0 0>, angularTarget=<0 0 0>, linearAxesMax=<0.5 0.5 0.5>, linearAxesMin=<-0.5 -0.5 -0.5>}
      localJointOffset=<-1 -0.25 -0.5>
     }]
    relativeSpaceBone {boneName=pack}
   }
   Output Pose(Result=Component To Local(ComponentPose=@n9)) {layerGroup=None}
  ]}
 EventGraph {nodes [
   Event BlueprintUpdateAnimation()
   n1 = GetGameTimeInSeconds()
   Set rotator(rotator:Rotator=MakeRotator(Pitch=float * float(A=@n1, B=-2.5))) -> (Output_Get:Rotator)
  ]}
]
variables [rotator {type=Rotator}]
